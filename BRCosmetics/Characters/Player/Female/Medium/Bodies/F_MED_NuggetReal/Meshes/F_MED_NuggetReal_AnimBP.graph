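# Anim Blueprint for the F_MED_NuggetReal cosmetic body, adding secondary motion to its pants, fringe and cord.
# It copies the parent mesh pose, runs a Control Rig, chains per-bone spring-based Anim Dynamics, then adds a yaw offset to the cord.
# The event graph drives Control Rig wave speed and toggle from skydive, dive and parachute states.
# It also sets GravityOverride and JumpFall, which the AnimGraph never reads.
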AnimBlueprint /BRCosmetics/Characters/Player/Female/Medium/Bodies/F_MED_NuggetReal/Meshes/F_MED_NuggetReal_AnimBP
target_skeleton=&/F_MED_NuggetReal_Skeleton
graphs [
 AnimGraph {nodes [
   n1 = Control Rig!(
    Source=Copy Pose From Mesh() {bUseAttachedParent=true}
    Alpha
    Skydive_WaveSpeed:double=Get WindSpeed() -> (WindSpeed:double)
    skydive_Wavealpha:bool=Get Wave_alpha() -> (Wave_alpha:bool)
   ) -> (Pose) {controlRigAssetReference {controlRigAsset=&/F_MED_NuggetReal_CtrlRig}}
   n2 = Anim Dynamics - Bone: dyn_lowerpants_a_l_01(ComponentPose=Local To Component(LocalPose=@n1), Alpha=1, ExternalForce="0,0,0") {
    angularDampingOverride=1
    angularSpringConstant=14000
    bAngularSpring=true
    bOverrideAngularDamping=true
    bUseGravityOverride=true
    gravityScale=0.35
    simulationSpace=RootRelative
    boundBone {boneName=dyn_lowerpants_a_l_01}
    componentAppliedLinearAccClamp=<5000 5000 5000>
    componentLinearAccScale=<0.7 0.7 0.7>
    componentLinearVelScale=<0.7 0.7 0.7>
    gravityOverride {z=-960}
    physicsBodyDefinitions [{
      boundBone {boneName=dyn_lowerpants_a_l_01}
      boxExtents=<20 28 28>
      constraintSetup {angularLimitsMax=<20 5 3>, angularLimitsMin=<-18 -5 -3>, angularTarget=<1 0 0>}
      localJointOffset=<3.9 0 0>
     }]
   }
   n3 = Anim Dynamics - Bone: dyn_lowerpants_b_l_01(ComponentPose=@n2, Alpha=1, ExternalForce="0,0,0") {
    angularDampingOverride=1
    angularSpringConstant=14000
    bAngularSpring=true
    bOverrideAngularDamping=true
    bUseGravityOverride=true
    gravityScale=0.35
    simulationSpace=RootRelative
    boundBone {boneName=dyn_lowerpants_b_l_01}
    componentAppliedLinearAccClamp=<5000 5000 5000>
    componentLinearAccScale=<0.6 0.6 0.6>
    componentLinearVelScale=<0.6 0.6 0.6>
    gravityOverride {z=-960}
    physicsBodyDefinitions [{
      boundBone {boneName=dyn_lowerpants_b_l_01}
      boxExtents=<17 17 17>
      constraintSetup {angularLimitsMax=<19 12 3>, angularLimitsMin=<-25 -12 -2>, angularTarget=<1 0 0>}
      localJointOffset=<4.3 0 0>
     }]
   }
   n4 = Anim Dynamics - Bone: dyn_lowerpants_a_r_01(ComponentPose=@n3, Alpha=1, ExternalForce="0,0,0") {
    angularDampingOverride=1
    angularSpringConstant=14000
    bAngularSpring=true
    bOverrideAngularDamping=true
    bUseGravityOverride=true
    gravityScale=0.35
    simulationSpace=RootRelative
    boundBone {boneName=dyn_lowerpants_a_r_01}
    componentAppliedLinearAccClamp=<5000 5000 5000>
    componentLinearAccScale=<0.7 0.7 0.7>
    componentLinearVelScale=<0.7 0.7 0.7>
    gravityOverride {z=-980}
    physicsBodyDefinitions [{
      boundBone {boneName=dyn_lowerpants_a_r_01}
      boxExtents=<20 30 30>
      constraintSetup {angularLimitsMax=<17 7 3>, angularLimitsMin=<-22 -7 -3>, angularTarget=<1 0 0>}
      localJointOffset=<3.75 0 0>
     }]
   }
   n5 = Anim Dynamics - Bone: dyn_lowerpants_b_r_01(ComponentPose=@n4, Alpha=1, ExternalForce="0,0,0") {
    angularDampingOverride=1
    angularSpringConstant=10000
    bAngularSpring=true
    bOverrideAngularDamping=true
    bUseGravityOverride=true
    gravityScale=0.35
    simulationSpace=RootRelative
    boundBone {boneName=dyn_lowerpants_b_r_01}
    componentAppliedLinearAccClamp=<5000 5000 5000>
    componentLinearAccScale=<0.6 0.6 0.6>
    componentLinearVelScale=<0.6 0.6 0.6>
    gravityOverride {z=-980}
    physicsBodyDefinitions [{
      boundBone {boneName=dyn_lowerpants_b_r_01}
      boxExtents=<17 17 17>
      constraintSetup {angularLimitsMax=<22 14 3>, angularLimitsMin=<-12 -14 -2>, angularTarget=<1 0 0>}
      localJointOffset=<4.3 0 0>
     }]
   }
   n6 = Anim Dynamics - Bone: dyn_fringeArm_a_l_01(ComponentPose=@n5, Alpha=1, ExternalForce="0,0,0") {
    angularDampingOverride=1
    angularSpringConstant=6000
    bAngularSpring=true
    bOverrideAngularDamping=true
    bUseGravityOverride=true
    gravityScale=0.35
    simulationSpace=RootRelative
    boundBone {boneName=dyn_fringeArm_a_l_01}
    componentAppliedLinearAccClamp=<5000 5000 5000>
    componentLinearAccScale=<0.5 0.5 0.5>
    componentLinearVelScale=<0.5 0.5 0.5>
    gravityOverride {z=-970}
    physicsBodyDefinitions [{
      boundBone {boneName=dyn_fringeArm_a_l_01}
      boxExtents=<20 20 20>
      constraintSetup {angularLimitsMax=<0 17 27>, angularLimitsMin=<0 -17 -27>, angularTarget=<1 0 0>}
      localJointOffset=<6.5 0 0>
     }]
   }
   n7 = Anim Dynamics - Bone: dyn_fringeArm_b_l_01(ComponentPose=@n6, Alpha=1, ExternalForce="0,0,0") {
    angularDampingOverride=1
    angularSpringConstant=11000
    bAngularSpring=true
    bOverrideAngularDamping=true
    bUseGravityOverride=true
    gravityScale=0.35
    simulationSpace=RootRelative
    boundBone {boneName=dyn_fringeArm_b_l_01}
    componentAppliedLinearAccClamp=<5000 5000 5000>
    componentLinearAccScale=<0.7 0.7 0.7>
    componentLinearVelScale=<0.7 0.7 0.7>
    gravityOverride {z=-980}
    physicsBodyDefinitions [{
      boundBone {boneName=dyn_fringeArm_b_l_01}
      boxExtents=<20 20 20>
      constraintSetup {angularLimitsMax=<5 19 20>, angularLimitsMin=<-5 -19 -20>, angularTarget=<1 0 0>}
      localJointOffset=<8.2 0 0>
     }]
   }
   n8 = Anim Dynamics - Bone: dyn_fringeArm_c_l_01(ComponentPose=@n7, Alpha=1, ExternalForce="0,0,0") {
    angularDampingOverride=1
    angularSpringConstant=6000
    bAngularSpring=true
    bOverrideAngularDamping=true
    bUseGravityOverride=true
    gravityScale=0.35
    simulationSpace=RootRelative
    boundBone {boneName=dyn_fringeArm_c_l_01}
    componentAppliedLinearAccClamp=<5000 5000 5000>
    componentLinearAccScale=<0.5 0.5 0.5>
    componentLinearVelScale=<0.5 0.5 0.5>
    gravityOverride {z=-970}
    physicsBodyDefinitions [{
      boundBone {boneName=dyn_fringeArm_c_l_01}
      boxExtents=<20 20 20>
      constraintSetup {angularLimitsMax=<2 17 30>, angularLimitsMin=<-2 -17 -30>, angularTarget=<1 0 0>}
      localJointOffset=<6 0 0>
     }]
   }
   n9 = Anim Dynamics - Bone: dyn_fringeArm_a_r_01(ComponentPose=@n8, Alpha=1, ExternalForce="0,0,0") {
    angularDampingOverride=1
    angularSpringConstant=9000
    bAngularSpring=true
    bOverrideAngularDamping=true
    bUseGravityOverride=true
    gravityScale=0.35
    simulationSpace=RootRelative
    boundBone {boneName=dyn_fringeArm_a_r_01}
    componentAppliedLinearAccClamp=<5000 5000 5000>
    componentLinearAccScale=<0.4 0.4 0.4>
    componentLinearVelScale=<0.4 0.4 0.4>
    gravityOverride {z=-920}
    physicsBodyDefinitions [{
      boundBone {boneName=dyn_fringeArm_a_r_01}
      boxExtents=<20 20 20>
      constraintSetup {angularLimitsMax=<10 17 32>, angularLimitsMin=<-10 -17 -32>, angularTarget=<1 0 0>}
      localJointOffset=<7 0 0>
     }]
   }
   n10 = Anim Dynamics - Bone: dyn_fringeArm_b_r_01(ComponentPose=@n9, Alpha=1, ExternalForce="0,0,0") {
    angularDampingOverride=1
    angularSpringConstant=12000
    bAngularSpring=true
    bOverrideAngularDamping=true
    bUseGravityOverride=true
    gravityScale=0.35
    simulationSpace=RootRelative
    boundBone {boneName=dyn_fringeArm_b_r_01}
    componentAppliedLinearAccClamp=<5000 5000 5000>
    componentLinearAccScale=<0.4 0.4 0.4>
    componentLinearVelScale=<0.4 0.4 0.4>
    gravityOverride {z=-920}
    physicsBodyDefinitions [{
      boundBone {boneName=dyn_fringeArm_b_r_01}
      boxExtents=<20 20 20>
      constraintSetup {angularLimitsMax=<0 19 18>, angularLimitsMin=<0 -19 -18>, angularTarget=<1 0 0>}
      localJointOffset=<6 0 0>
     }]
   }
   n11 = Anim Dynamics - Bone: dyn_fringeArm_c_r_01(ComponentPose=@n10, Alpha=1, ExternalForce="0,0,0") {
    angularDampingOverride=1
    angularSpringConstant=9000
    bAngularSpring=true
    bOverrideAngularDamping=true
    bUseGravityOverride=true
    gravityScale=0.35
    simulationSpace=RootRelative
    boundBone {boneName=dyn_fringeArm_c_r_01}
    componentAppliedLinearAccClamp=<5000 5000 5000>
    componentLinearAccScale=<0.4 0.4 0.4>
    componentLinearVelScale=<0.4 0.4 0.4>
    gravityOverride {z=-920}
    physicsBodyDefinitions [{
      boundBone {boneName=dyn_fringeArm_c_r_01}
      boxExtents=<20 20 20>
      constraintSetup {angularLimitsMax=<10 17 32>, angularLimitsMin=<-10 -17 -32>, angularTarget=<1 0 0>}
      localJointOffset=<7 0 0>
     }]
   }
   n12 = Anim Dynamics - Bone: dyn_fringeLeg_a_l_01!(ComponentPose=@n11, Alpha=1, GravityOverride:Vector="0,0,-950", ExternalForce="0,0,0") -> (Pose) {
    angularDampingOverride=1
    angularSpringConstant=1000
    bAngularSpring=true
    bOverrideAngularDamping=true
    bUseGravityOverride=true
    gravityScale=0.35
    simulationSpace=RootRelative
    boundBone {boneName=dyn_fringeLeg_a_l_01}
    componentAppliedLinearAccClamp=<5000 5000 5000>
    componentLinearAccScale=<0.3 0.3 0.3>
    componentLinearVelScale=<0.3 0.3 0.3>
    gravityOverride {z=-950}
    physicsBodyDefinitions [{
      boundBone {boneName=dyn_fringeLeg_a_l_01}
      boxExtents=<18 18 18>
      constraintSetup {angularLimitsMax=<26 28 7>, angularLimitsMin=<-26 -25 -7>, angularTarget=<1 0 0>}
      localJointOffset=<5 0 0>
     }]
    simSpaceSettings {maxAngularAcceleration=50, maxAngularVelocity=10, simSpaceAngularAlpha=0.7}
   }
   n13 = Anim Dynamics - Bone: dyn_fringeLeg_b_l_01!(ComponentPose=@n12, Alpha=1, GravityOverride:Vector="0,0,-920", ExternalForce="0,0,0") -> (Pose) {
    angularDampingOverride=1
    angularSpringConstant=1000
    bAngularSpring=true
    bOverrideAngularDamping=true
    bUseGravityOverride=true
    gravityScale=0.35
    simulationSpace=RootRelative
    boundBone {boneName=dyn_fringeLeg_b_l_01}
    componentAppliedLinearAccClamp=<5000 5000 5000>
    componentLinearAccScale=<0.3 0.3 0.3>
    componentLinearVelScale=<0.3 0.3 0.3>
    gravityOverride {z=-920}
    physicsBodyDefinitions [{
      boundBone {boneName=dyn_fringeLeg_b_l_01}
      boxExtents=<18 18 18>
      constraintSetup {angularLimitsMax=<26 28 14>, angularLimitsMin=<-26 -25 -14>, angularTarget=<1 0 0>}
      localJointOffset=<9 0 0>
     }]
    simSpaceSettings {maxAngularAcceleration=50, maxAngularVelocity=10, simSpaceAngularAlpha=0.7}
   }
   n14 = Anim Dynamics - Bone: dyn_fringeLeg_c_l_01!(ComponentPose=@n13, Alpha=1, GravityOverride:Vector="0,0,-920", ExternalForce="0,0,0") -> (Pose) {
    angularDampingOverride=1
    angularSpringConstant=3000
    bAngularSpring=true
    bOverrideAngularDamping=true
    bUseGravityOverride=true
    gravityScale=0.35
    simulationSpace=RootRelative
    boundBone {boneName=dyn_fringeLeg_c_l_01}
    componentAppliedLinearAccClamp=<5000 5000 5000>
    componentLinearAccScale=<0.3 0.3 0.3>
    componentLinearVelScale=<0.3 0.3 0.3>
    gravityOverride {z=-920}
    physicsBodyDefinitions [{
      boundBone {boneName=dyn_fringeLeg_c_l_01}
      boxExtents=<18 18 18>
      constraintSetup {angularLimitsMax=<26 28 17>, angularLimitsMin=<-26 -25 -17>, angularTarget=<1 0 0>}
      localJointOffset=<7 0 0>
     }]
    simSpaceSettings {maxAngularAcceleration=50, maxAngularVelocity=10, simSpaceAngularAlpha=0.7}
   }
   n15 = Anim Dynamics - Bone: dyn_fringeLeg_d_l_01!(ComponentPose=@n14, Alpha=1, GravityOverride:Vector="0,0,-920", ExternalForce="0,0,0") -> (Pose) {
    angularDampingOverride=1
    angularSpringConstant=1000
    bAngularSpring=true
    bOverrideAngularDamping=true
    bUseGravityOverride=true
    gravityScale=0.35
    simulationSpace=RootRelative
    boundBone {boneName=dyn_fringeLeg_d_l_01}
    componentAppliedLinearAccClamp=<5000 5000 5000>
    componentLinearAccScale=<0.3 0.3 0.3>
    componentLinearVelScale=<0.3 0.3 0.3>
    gravityOverride {z=-920}
    physicsBodyDefinitions [{
      boundBone {boneName=dyn_fringeLeg_d_l_01}
      boxExtents=<18 18 18>
      constraintSetup {angularLimitsMax=<26 28 14>, angularLimitsMin=<-26 -25 -14>, angularTarget=<1 0 0>}
      localJointOffset=<5 0 0>
     }]
    simSpaceSettings {maxAngularAcceleration=50, maxAngularVelocity=10, simSpaceAngularAlpha=0.7}
   }
   n16 = Anim Dynamics - Bone: dyn_fringeLeg_e_l_01!(ComponentPose=@n15, Alpha=1, GravityOverride:Vector="0,0,-950", ExternalForce="0,0,0") -> (Pose) {
    angularDampingOverride=1
    angularSpringConstant=1000
    bAngularSpring=true
    bOverrideAngularDamping=true
    bUseGravityOverride=true
    gravityScale=0.35
    simulationSpace=RootRelative
    boundBone {boneName=dyn_fringeLeg_e_l_01}
    componentAppliedLinearAccClamp=<5000 5000 5000>
    componentLinearAccScale=<0.3 0.3 0.3>
    componentLinearVelScale=<0.3 0.3 0.3>
    gravityOverride {z=-950}
    physicsBodyDefinitions [{
      boundBone {boneName=dyn_fringeLeg_e_l_01}
      boxExtents=<18 18 18>
      constraintSetup {angularLimitsMax=<26 28 17>, angularLimitsMin=<-26 -25 -17>, angularTarget=<1 0 0>}
      localJointOffset=<5 0 0>
     }]
    simSpaceSettings {maxAngularAcceleration=50, maxAngularVelocity=10, simSpaceAngularAlpha=0.7}
   }
   n17 = Anim Dynamics - Bone: dyn_fringeLeg_a_r_01!(ComponentPose=@n16, Alpha=1, GravityOverride:Vector="0,0,-920", ExternalForce="0,0,0") -> (Pose) {
    angularDampingOverride=1
    angularSpringConstant=1000
    bAngularSpring=true
    bOverrideAngularDamping=true
    bUseGravityOverride=true
    gravityScale=0.35
    simulationSpace=RootRelative
    boundBone {boneName=dyn_fringeLeg_a_r_01}
    componentAppliedLinearAccClamp=<5000 5000 5000>
    componentLinearAccScale=<0.3 0.3 0.3>
    componentLinearVelScale=<0.3 0.3 0.3>
    gravityOverride {z=-920}
    physicsBodyDefinitions [{
      boundBone {boneName=dyn_fringeLeg_a_r_01}
      boxExtents=<18 18 18>
      constraintSetup {angularLimitsMax=<26 28 8>, angularLimitsMin=<-26 -25 -7>, angularTarget=<1 0 0>}
      localJointOffset=<5 0 0>
     }]
    simSpaceSettings {maxAngularAcceleration=50, maxAngularVelocity=10, simSpaceAngularAlpha=0.7}
   }
   n18 = Anim Dynamics - Bone: dyn_fringeLeg_b_r_01!(ComponentPose=@n17, Alpha=1, GravityOverride:Vector="0,0,-920", ExternalForce="0,0,0") -> (Pose) {
    angularDampingOverride=1
    angularSpringConstant=3000
    bAngularSpring=true
    bOverrideAngularDamping=true
    bUseGravityOverride=true
    gravityScale=0.35
    simulationSpace=RootRelative
    boundBone {boneName=dyn_fringeLeg_b_r_01}
    componentAppliedLinearAccClamp=<5000 5000 5000>
    componentLinearAccScale=<0.3 0.3 0.3>
    componentLinearVelScale=<0.3 0.3 0.3>
    gravityOverride {z=-920}
    physicsBodyDefinitions [{
      boundBone {boneName=dyn_fringeLeg_b_r_01}
      boxExtents=<18 18 18>
      constraintSetup {angularLimitsMax=<26 28 16>, angularLimitsMin=<-26 -25 -16>, angularTarget=<1 0 0>}
      localJointOffset=<8 0 0>
     }]
    simSpaceSettings {maxAngularAcceleration=50, maxAngularVelocity=10, simSpaceAngularAlpha=0.7}
   }
   n19 = Anim Dynamics - Bone: dyn_fringeLeg_c_r_01!(ComponentPose=@n18, Alpha=1, GravityOverride:Vector="0,0,-920", ExternalForce="0,0,0") -> (Pose) {
    angularDampingOverride=1
    angularSpringConstant=1000
    bAngularSpring=true
    bOverrideAngularDamping=true
    bUseGravityOverride=true
    gravityScale=0.35
    simulationSpace=RootRelative
    boundBone {boneName=dyn_fringeLeg_c_r_01}
    componentAppliedLinearAccClamp=<5000 5000 5000>
    componentLinearAccScale=<0.3 0.3 0.3>
    componentLinearVelScale=<0.3 0.3 0.3>
    gravityOverride {z=-920}
    physicsBodyDefinitions [{
      boundBone {boneName=dyn_fringeLeg_c_r_01}
      boxExtents=<18 18 18>
      constraintSetup {angularLimitsMax=<26 28 16>, angularLimitsMin=<-26 -25 -16>, angularTarget=<1 0 0>}
      localJointOffset=<7 0 0>
     }]
    simSpaceSettings {maxAngularAcceleration=50, maxAngularVelocity=10, simSpaceAngularAlpha=0.7}
   }
   n20 = Anim Dynamics - Bone: dyn_fringeLeg_d_r_01!(ComponentPose=@n19, Alpha=1, GravityOverride:Vector="0,0,-920", ExternalForce="0,0,0") -> (Pose) {
    angularDampingOverride=1
    angularSpringConstant=1000
    bAngularSpring=true
    bOverrideAngularDamping=true
    bUseGravityOverride=true
    gravityScale=0.35
    simulationSpace=RootRelative
    boundBone {boneName=dyn_fringeLeg_d_r_01}
    componentAppliedLinearAccClamp=<5000 5000 5000>
    componentLinearAccScale=<0.3 0.3 0.3>
    componentLinearVelScale=<0.3 0.3 0.3>
    gravityOverride {z=-920}
    physicsBodyDefinitions [{
      boundBone {boneName=dyn_fringeLeg_d_r_01}
      boxExtents=<18 18 18>
      constraintSetup {angularLimitsMax=<26 28 8>, angularLimitsMin=<-26 -25 -7>, angularTarget=<1 0 0>}
      localJointOffset=<5 0 0>
     }]
    simSpaceSettings {maxAngularAcceleration=50, maxAngularVelocity=10, simSpaceAngularAlpha=0.7}
   }
   n21 = Anim Dynamics - Bone: dyn_fringeLeg_e_r_01!(ComponentPose=@n20, Alpha=1, GravityOverride:Vector="0,0,-980", ExternalForce="0,0,0") -> (Pose) {
    angularDampingOverride=1
    angularSpringConstant=1000
    bAngularSpring=true
    bOverrideAngularDamping=true
    bUseGravityOverride=true
    gravityScale=0.35
    simulationSpace=RootRelative
    boundBone {boneName=dyn_fringeLeg_e_r_01}
    componentAppliedLinearAccClamp=<5000 5000 5000>
    componentLinearAccScale=<0.3 0.3 0.3>
    componentLinearVelScale=<0.3 0.3 0.3>
    gravityOverride {z=-980}
    physicsBodyDefinitions [{
      boundBone {boneName=dyn_fringeLeg_e_r_01}
      boxExtents=<18 18 18>
      constraintSetup {angularLimitsMax=<26 28 19>, angularLimitsMin=<-26 -25 -19>, angularTarget=<1 0 0>}
      localJointOffset=<5 0 0>
     }]
    simSpaceSettings {maxAngularAcceleration=50, maxAngularVelocity=10, simSpaceAngularAlpha=0.7}
   }
   n22 = Anim Dynamics - Bone: dyn_cord(ComponentPose=@n21, Alpha=1, ExternalForce="0,0,0") {
    angularDampingOverride=1
    angularSpringConstant=8000
    bAngularSpring=true
    bOverrideAngularDamping=true
    bUseGravityOverride=true
    gravityScale=0.35
    simulationSpace=RootRelative
    boundBone {boneName=dyn_cord}
    componentAppliedLinearAccClamp=<5000 5000 5000>
    componentLinearAccScale=<0.5 0.5 0.5>
    componentLinearVelScale=<0.5 0.5 0.5>
    gravityOverride {z=-950}
    physicsBodyDefinitions [{
      boundBone {boneName=dyn_cord}
      boxExtents=<22 21 21>
      constraintSetup {angularLimitsMax=<0 15 50>, angularLimitsMin=<0 -15 -19>, angularTarget=<1 0 0>}
      localJointOffset=<5 0 0>
     }]
   }
   n23 = "Transform (Modify) Bone - Bone: dyn_cord"!(ComponentPose=@n22, bAlphaBoolEnabled:bool=True, Translation="0,0,0", Rotation="0,33,0", Scale="1,1,1") -> (Pose) {
    alphaInputType=Bool
    rotationMode=BMM_Additive
    rotationSpace=BCS_BoneSpace
    alphaBoolBlend {blendInTime=0.45, blendOutTime=0.15}
    boneToModify {boneName=dyn_cord}
    rotation {yaw=33}
   }
   Output Pose(Result=Component To Local(ComponentPose=@n23)) {layerGroup=None}
  ]}
 EventGraph {nodes [
   Event BlueprintUpdateAnimation()
   > n1 = Sequence!(execute) -> (then_0, then_1, then_2:Exec)
   > n2 = "Gravity Override State Function (Parameter Asset) (Nativized)"(
    CharacterPartAnimInstance=Self-Reference()
    GameTimeInSeconds=GetGameTimeInSeconds()
    Parameters=&/F_MED_NuggetReal_Parameter
   )
   > Set GravityOverride(GravityOverride=@n2)
   Set JumpFall(execute=@n1.then_1, JumpFall:bool=OR Boolean(A=Get bIsJumping(), B=Get bIsFalling())) -> (Output_Get:bool)
   n3 = Branch(execute=@n1.then_2, Condition=OR Boolean(A=Get bIsSkyDiving(), B=Get bIsParachuteOpened()))
   > n4 = Branch(Condition=Get bIsSkydiveDiveMode())
   > Set WindSpeed(WindSpeed:double=12) -> (Output_Get:double)
   > Set Wave_alpha(Wave_alpha:bool=true) -> (Output_Get:bool)
   Set WindSpeed(execute=@n4.else, WindSpeed=7.25)
   > Set Wave_alpha(Wave_alpha=true)
   Set WindSpeed(execute=@n3.else, WindSpeed=0)
   > Set Wave_alpha(Wave_alpha=false)
  ]}
]
variables [
 GravityOverride {type=Vector}
 JumpFall {type=boolean}
 Wave_alpha {type=boolean}
 WindSpeed {type=number}
]
