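# Physics asset for the F_MED_NuggetReal female character's face accessory part, setting up secondary motion for hair and ribbons.
# A five-link capsule chain hangs from a tiny base hair body with narrow angular limits, and two ribbons branch off the fourth link.
# Spine, neck, upper arm and hat capsules are joined by angle-locked constraints, apparently as collision volumes.
# Only the left ribbon has twist locked, unlike the right.
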PhysicsAsset /BRCosmetics/Characters/Player/Female/Medium/Bodies/F_MED_NuggetReal/Meshes/Parts/F_MED_NuggetReal_FaceAcc_What_Physics
bodies {
 base_hair_01 [{shapeType=Capsule, length=0.505, radius=0.505, shapeName=base_hair_01_capsule, rotation=<rot 0 -90 0>}]
 dyn_hair_01 [{shapeType=Capsule, length=4, radius=3, shapeName=dyn_hair_01_capsule, center=<4.9216 0 0>, rotation=<rot 0 90 90>}]
 dyn_hair_02 [{shapeType=Capsule, length=2, radius=3, shapeName=dyn_hair_02_capsule, center=<3.1489 0 0>, rotation=<rot 90 0 0>}]
 dyn_hair_03 [{shapeType=Capsule, length=1.5, radius=3, shapeName=dyn_hair_03_capsule, center=<3.7 0 0.7822>, rotation=<rot 90 0 0>}]
 dyn_hair_04 [{shapeType=Capsule, length=3.5, radius=3.5, shapeName=dyn_hair_04_capsule, center=<4.2 0 1.302>, rotation=<rot 80.0001 180 180>}]
 dyn_hair_05 [{shapeType=Capsule, length=3.5, radius=3, shapeName=dyn_hair_05_capsule, center=<3.7291 0 0.3115>, rotation=<rot 90 0 0>}]
 dyn_ribbon_l_01 [{shapeType=Capsule, length=3.5, radius=1.9, shapeName=dyn_ribbon_l_01_capsule, center=<2.78 0 1.85>, rotation=<rot 80 -179.9995 -179.9996>}]
 dyn_ribbon_r_01 [{shapeType=Capsule, length=3.5, radius=1.9, shapeName=dyn_ribbon_r_01_capsule, center=<2.78 0 1.85>, rotation=<rot -80 0 0>}]
 hat [
  {shapeType=Capsule, length=50, radius=12.5, shapeName=hat_capsule, center=<8.6149 -26.0005 0>, rotation=<rot 0 -90 0>}
  {shapeType=Capsule, length=69.5, radius=13.5, shapeName=hat_capsule, center=<13.5445 -2.4487 0>}
 ]
 neck_01 [{shapeType=Capsule, length=72, radius=6, shapeName=neck_01_capsule, center=<4.709 2.2609 0>, rotation=<rot 0 -90 0>}]
 spine_04 [{shapeType=Capsule, length=5.5, radius=12, shapeName=spine_04_capsule, center=<0 0.9538 0>, rotation=<rot 0 -90 90>}]
 spine_05 [{shapeType=Capsule, length=64, radius=9.5, shapeName=spine_05_capsule, center=<4.8352 2.214 0>, rotation=<rot 0 90 0>}]
 upperarm_l [{shapeType=Capsule, length=19, radius=5.5, shapeName=upperarm_l_capsule, center=<8.1608 0.4809 0>, rotation=<rot 80 0.011 -0.0003>}]
 upperarm_r [{shapeType=Capsule, length=19, radius=5.5, shapeName=upperarm_r_capsule, center=<-7.948 -0.1542 0>, rotation=<rot 80 0 0>}]
}
constraints [
 {bone1Name=dyn_hair_01, bone2Name=base_hair_01, swing1LimitDegrees=12, swing2LimitDegrees=20, twistLimitDegrees=2}
 {
  bone1Name=spine_05
  bone2Name=spine_04
  swing1LimitDegrees=45
  swing1Motion=Locked
  swing2LimitDegrees=45
  swing2Motion=Locked
  twistLimitDegrees=45
  twistMotion=Locked
 }
 {
  bone1Name=neck_01
  bone2Name=spine_05
  swing1LimitDegrees=45
  swing1Motion=Locked
  swing2LimitDegrees=45
  swing2Motion=Locked
  twistLimitDegrees=45
  twistMotion=Locked
 }
 {bone1Name=dyn_hair_02, bone2Name=dyn_hair_01, swing1LimitDegrees=18, swing2LimitDegrees=23, twistLimitDegrees=2}
 {bone1Name=dyn_hair_03, bone2Name=dyn_hair_02, swing1LimitDegrees=18, swing2LimitDegrees=23, twistLimitDegrees=2}
 {bone1Name=dyn_hair_04, bone2Name=dyn_hair_03, swing1LimitDegrees=18, swing2LimitDegrees=23, twistLimitDegrees=2}
 {bone1Name=dyn_hair_05, bone2Name=dyn_hair_04, swing1LimitDegrees=18, swing2LimitDegrees=23, twistLimitDegrees=2}
 {bone1Name=dyn_ribbon_l_01, bone2Name=dyn_hair_04, swing1LimitDegrees=15, swing2LimitDegrees=25, twistLimitDegrees=2, twistMotion=Locked}
 {bone1Name=dyn_ribbon_r_01, bone2Name=dyn_hair_04, swing1LimitDegrees=15, swing2LimitDegrees=25, twistLimitDegrees=2}
 {
  bone1Name=upperarm_r
  bone2Name=spine_05
  swing1LimitDegrees=45
  swing1Motion=Locked
  swing2LimitDegrees=45
  swing2Motion=Locked
  twistLimitDegrees=45
  twistMotion=Locked
 }
 {
  bone1Name=upperarm_l
  bone2Name=spine_05
  swing1LimitDegrees=45
  swing1Motion=Locked
  swing2LimitDegrees=45
  swing2Motion=Locked
  twistLimitDegrees=45
  twistMotion=Locked
 }
 {
  bone1Name=hat
  bone2Name=neck_01
  swing1LimitDegrees=45
  swing1Motion=Locked
  swing2LimitDegrees=45
  swing2Motion=Locked
  twistLimitDegrees=45
  twistMotion=Locked
 }
]
