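# Animation Blueprint for the female NuggetReal face-accessory cosmetic part, copying the parent mesh pose and adding secondary hair and accessory motion.
# With a qualifying backbling, bone constraints blend hair between head and spine, with an additive base offset.
# Otherwise, a Stage0 variant tag selects zero-gravity RigidBody simulation; other variants use base-bone-space simulation with gravity from a parameter-driven state function.
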AnimBlueprint /BRCosmetics/Characters/Player/Female/Medium/Bodies/F_MED_NuggetReal/Meshes/Parts/F_MED_NuggetReal_FaceAcc_AnimBP
target_skeleton=&/F_MED_NuggetReal_FaceAcc_Skeleton
graphs [
 AnimGraph {nodes [
   n1 = RigidBody!(
    ComponentPose=Local To Component(LocalPose=Use cached pose 'Input Pose'())
    Alpha
    OverrideWorldGravity:Vector="0,0,0"
    ExternalForce
    SimSpaceSettings:Sim Space Settings Structure=(WorldAlpha=0,VelocityScaleZ=1,DampingAlpha=1,MaxLinearVelocity=10000,MaxAngularVelocity=10000,MaxLinearAcceleration=10000,MaxAngularAcceleration=10000,ExternalLinearDragV=(X=0,Y=0,Z=0),ExternalLinearVelocity=(X=0,Y=0,Z=0),ExternalAngularVelocity=(X=0,Y=0,Z=0))
   ) -> (Pose) {
    bOverrideWorldGravity=true
    overridePhysicsAsset=&/F_MED_NuggetReal_FaceAcc_What_Physics
    baseBoneRef {boneName=neck_01}
   }
   n2 = Component To Local(ComponentPose=@n1)
   n3 = RigidBody(
    ComponentPose=Local To Component(LocalPose=Use cached pose 'Input Pose'())
    OverrideWorldGravity=Get GravityOverride()
    SimSpaceSettings=(WorldAlpha=1,VelocityScaleZ=1,DampingAlpha=1,MaxLinearVelocity=250,MaxAngularVelocity=3,MaxLinearAcceleration=800,MaxAngularAcceleration=200,ExternalLinearDragV=(X=0.4,Y=0.4,Z=0.4),ExternalLinearVelocity=(X=0.2,Y=0.2,Z=0.2),ExternalAngularVelocity=(X=0,Y=0,Z=0))
   ) {
    bOverrideWorldGravity=true
    overridePhysicsAsset=&/F_MED_NuggetReal_FaceAcc_Physics
    simulationSpace=BaseBoneSpace
    baseBoneRef {boneName=neck_02}
    simSpaceSettings {
     maxAngularAcceleration=200
     maxAngularVelocity=3
     maxLinearAcceleration=800
     maxLinearVelocity=250
     worldAlpha=1
     externalLinearDragV=<0.4 0.4 0.4>
     externalLinearVelocity=<0.2 0.2 0.2>
    }
   }
   n4 = Component To Local(ComponentPose=@n3)
   n5 = Get Hat vis() -> (Hat vis:bool)
   n6 = "Constraint - dyn_hair_03 (PP)"!(ComponentPose=Local To Component(LocalPose=Use cached pose 'Input Pose'()), Alpha=1, ConstraintWeights_0=0.5, ConstraintWeights_1:float=0.5) -> (Pose) {
    boneToModify {boneName=dyn_hair_03}
    constraintSetup [{targetBone {boneName=head}}, {targetBone {boneName=spine_05}}]
    constraintWeights [1,1]
   }
   n7 = "Constraint - dyn_hair_04 (PP)"!(ComponentPose=@n6, Alpha=1, ConstraintWeights_0=0.3, ConstraintWeights_1:float=0.7) -> (Pose) {
    boneToModify {boneName=dyn_hair_04}
    constraintSetup [{targetBone {boneName=head}}, {targetBone {boneName=spine_05}}]
    constraintWeights [1,1]
   }
   n8 = "Constraint - dyn_hair_05 (PP)"!(ComponentPose=@n7, Alpha=1, ConstraintWeights_0=0.2, ConstraintWeights_1:float=0.8) -> (Pose) {
    boneToModify {boneName=dyn_hair_05}
    constraintSetup [{targetBone {boneName=head}}, {targetBone {boneName=spine_05}}]
    constraintWeights [1,1]
   }
   n9 = "Transform (Modify) Bone - Bone: base_hair_01"(ComponentPose=@n8, Alpha=1, Translation="0,0,0", Rotation="-9.83,0,0.005", Scale="1,1,1") {
    rotationMode=BMM_Additive
    rotationSpace=BCS_BoneSpace
    translationMode=BMM_Additive
    translationSpace=BCS_BoneSpace
    boneToModify {boneName=base_hair_01}
    rotation {pitch=-9.83, roll=0.005}
   }
   Output Pose(Result=Blend Poses by bool(
     BlendPose_0=Component To Local(ComponentPose=@n9)
     BlendPose_1=Blend Poses by bool(BlendPose_0=@n2, BlendPose_1=@n4, bActiveValue=@n5)
     bActiveValue=Get isWearingPack() -> (isWearingPack:bool)
    )) {layerGroup=None}
   Save cached pose 'Input Pose'(Pose=Copy Pose From Mesh() {bUseAttachedParent=true}) {cacheName=Input Pose}
  ]}
 EventGraph {nodes [
   n1 = Cast To FortPlayerPawn(Object=TryGetPawnOwner())
   n2 = Get Mesh(self=@n1)
   n3 = BreakRotator(InRot=BreakTransform(InTransform=GetDeltaTransformFromRefPose(self=@n2, BoneName=head, BaseName=spine_05)).Rotation)
   > Set HatRotate(HatRotate:Rotator=MakeRotator(Yaw=MapRangeClamped(Value=@n3.Yaw, InRangeA=-30, InRangeB=0, OutRangeA=45, OutRangeB=0))) -> (Output_Get:Rotator)
   Event BlueprintUpdateAnimation()
   > n4 = Sequence!(execute) -> (then_0, then_1, then_2:Exec)
   n5 = "Gravity Override State Function (Parameter Asset) (Nativized)"(
    execute=@n4.then_1
    CharacterPartAnimInstance=Self-Reference()
    GameTimeInSeconds=GetGameTimeInSeconds()
    DebugJointName=neck_02
    Parameters=&/F_MED_NuggetReal_FaceAcc_Parameter
   )
   > Set GravityOverride(GravityOverride=@n5)
   > Set RBAN Settings Out(RBAN Settings Out=@n5.OutRigidBodyAnimNodeParameters)
   n6 = Branch(execute=@n4.then_2, Condition=CharacterItemDefinitionHasAnyMetaTag(Tags=Make Array("[0]"=(TagName="Cosmetics.Variant.Property.Stage0"))))
   > Set Hat vis(Hat vis:bool=true) -> (Output_Get:bool)
   Set Hat vis(execute=@n6.else, Hat vis=false)
   Event OnCharacterCustomizationLoaded()
   n7 = CheckForBackbling()
   > Set isWearingPack(isWearingPack:bool=OR Boolean!(A=@n7, B=@n7.bIsBackblingACape, C:bool=@n7.bIsBackblingLong) -> (ReturnValue)) -> (Output_Get:bool)
  ]}
]
variables [
 GravityOverride {type=Vector}
 Hat vis {type=boolean}
 HatRotate {type=Rotator}
 isWearingPack {type=boolean}
 Loadout {type=CosmeticLoadout, default {slots []}}
 RBAN Settings Out {type=FortRigidBodyAnimNodeParameters}
]
